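# Niagara effect for the back bling detector skin in a Season 24 narrative quest, playing a short blue flash burst.
# Six active emitters fire staggered bursts: disk, back glow, bright glow with a light renderer, flare, halo and sparks, shaped by color and size curves.
# Disabled emitters hold a debug sprite and skeletal-mesh-attached green and blue indicator lights, driven by user toggle and blink-speed parameters, with audio on blue blinks.
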NiagaraSystem /Game/Animation/Props/Gadgets_Backbling_Detector/Skin/Detector_Ch4S2/FX/NS_Quest_S24_Narrative_Q3_BackSensor
dependencies {
 usedDataInterfaces [
  &NiagaraDataInterfaceAudioPlayer
  &NiagaraDataInterfaceColorCurve
  &NiagaraDataInterfaceCurve
  &NiagaraDataInterfaceVector2DCurve
 ]
 usedDynamicInputs [
  &ClampFloat
  &Subtract_Float
  &VectorLength
  &Subtract_Vector
  /Niagara/DynamicInputs/Helpers/SimulationPosition
  &DotProduct
  &NormalizeVector
  &Multiply_Vector
  &ConvertPositionToVector
 ]
 usedModules [
  &SystemState
  &self:SystemSpawnScript.SystemScriptSource.SystemScriptGraph.NiagaraNodeAssignment_0.SetVariables_#1
  &self:SystemSpawnScript.SystemScriptSource.SystemScriptGraph.NiagaraNodeAssignment_1.SetVariables_#2
 ]
}
emitters {
 BlueLight {
  bEnabled=false
  emitterSpawnScript [&SampleSkeletalMesh(
    AdditionalYawPitchRoll=false
    Bone / Socket Index=1
    Bone Sampling Mode:ENiagaraBoneSamplingMode=::NewEnumerator4
    Interpolate Mesh Sampling=false
    Position Offset Coordinate Space:ENiagaraCoordinateSpace=::Local
    Sampled Position Coordinate Space:ENiagaraCoordinateSpace=::Local
    Skeletal Mesh:NiagaraDataInterfaceSkeletalMesh=$User.SkeletalMesh
   )]
  emitterUpdateScript [
   &EmitterState(
    Inactive Response=::NewEnumerator1
    Life Cycle Mode=::NewEnumerator1
    Loop Duration=&Divide_Float(A=1, B=$User.BlinkSpeed)
    Recalculate Duration Each Loop=true
    UseLoopDelay=false
   )
   &SpawnBurst_Instantaneous(Spawn Group=1, Use Spawn Probability=false, UseLoopCountLimit=false) {module=SpawnBurst_Instantaneous001}
   /SetVariables(Emitter.IsEnabled:NiagaraBool=&SetBoolByFloatComparison(A=$User.BlinkSpeed, B=0, Comparison Type=::NewEnumerator5)) {isSetParameters=true}
  ]
  particleSpawnScript [
   &InitializeParticle(
    Color=<rgba 0 0.2 1 1>
    Lifetime=0.133
    Position Mode=::NewEnumerator0
    Sprite Size Mode=::NewEnumerator3
    Uniform Sprite Size=9
    Write Material Random=false
   )
   /Niagara/Modules/Audio/PlayAudio(
    Audio Playback Enabled:NiagaraBool=$Emitter.IsEnabled
    Audio Playback Position:NiagaraPosition=$Emitter.SampleSkeletalMesh.SampledPosition
    Audio Player=DataInterface() {propertyValues {
      Attenuation=null
      Concurrency=null
      MaxPlaysPerTick=10
      SoundToPlay=/BRCosmetics/Sounds/Other/Quests/InvestigateTheReadings/Spot_Backbling_Detector_Light_Flash_Cue
      bAllowLoopingOneShotSounds=false
      bLimitPlaysPerTick=true
      bOnlyActiveDuringGameplay=false
      bStopWhenComponentIsDestroyed=true
      ConfigurationUserParameter {parameter {name=None, typeDefHandle {registeredTypeIndex=1}, varData []}}
      ParameterNames []
     }}
    Audio Position Coordinate Space:ENiagaraCoordinateSpace=::Local
    Pitch:NiagaraFloat=&Lerp_Float(A=0.8, B=1.5, Alpha=0)
   )
  ]
  particleUpdateScript [
   &ParticleState()
   &ScaleColor(ScaleA=false, ScaleRGB=false)
   &CameraOffset(Camera Offset Amount=9)
  ]
  rendererClasses [&NiagaraSpriteRendererProperties, &NiagaraMeshRendererProperties]
  renderers [
   {rendererClass=&NiagaraSpriteRendererProperties}
   {rendererClass=&NiagaraMeshRendererProperties, rendererIndex=1}
  ]
 }
 Disk {
  emitterUpdateScript [
   &EmitterState()
   &SpawnBurst_Instantaneous(Spawn Time=0.05, Use Spawn Probability=false, UseLoopCountLimit=true)
  ]
  particleSpawnScript [
   &InitializeParticle(
    Color Mode=::NewEnumerator0
    Lifetime=0.22
    Sprite Rotation Mode=::NewEnumerator2
    Sprite Size Mode=::NewEnumerator3
    Uniform Sprite Size=80
    UsePositionOffset=false
    Write Material Random=true
   )
   &CameraOffset(Camera Offset Amount=15)
  ]
  particleUpdateScript [
   &ParticleState()
   &ScaleSpriteSize(
    Scale Sprite Size Mode=::NewEnumerator0
    Uniform Scale Factor=&FloatFromCurve(FloatCurve=DataInterface() {propertyValues {Curve {keys [
         {arriveTangent=-0.0785, interpMode=RCIM_Cubic, leaveTangent=-0.0785, tangentMode=RCTM_User, time=0.15, value=1}
         {arriveTangent=-1.2137, interpMode=RCIM_Cubic, leaveTangent=-1.2137, tangentMode=RCTM_User, time=1, value=0.1}
        ]}}})
   ) {module=ScaleSpriteSize001}
   /SetVariables(Particles.ColorBrightGlowFlare:LinearColor=&MakeLinearColorFromVectorAndFloat(
     "Vector (RGB)"=&MakeVectorFromLinearColorRGB(Color=&ColorFromCurve(ColorCurve=DataInterface() {propertyValues {
         AlphaCurve {keys []}
         BlueCurve {keys [<at 0.0847 1.6>,<at 0.8655 12>]}
         GreenCurve {keys [<at 0.0847 0.9716>,<at 0.8655 8.1906>]}
         RedCurve {keys [<at 0.0847 0.5227>,<at 0.8655 0.84>]}
        }}))
     "Float (Alpha)"=&FloatFromCurve(
      FloatCurve=DataInterface() {propertyValues {Curve {keys [
          {arriveTangent=-0.1613, interpMode=RCIM_Cubic, leaveTangent=-0.1613, tangentMode=RCTM_User, time=0.2, value=0}
          {arriveTangent=2.1957, interpMode=RCIM_Cubic, leaveTangent=2.1957, tangentMode=RCTM_User, time=0.9, value=1}
          {arriveTangent=-13.9699, interpMode=RCIM_Cubic, leaveTangent=-13.9699, tangentMode=RCTM_User, time=1, value=0}
         ]}}}
      Scale Curve=0.2
     )
    )) {isSetParameters=true}
   /SetVariables(Particles.SpriteSizeGlowFlare:Vector2f=&Vector2DFromFloat(Value=&FloatFromCurve(FloatCurve=DataInterface() {propertyValues {Curve {keys [<at 0 1>,<at 1 0.7>]}}}, Scale Curve=50))) {isSetParameters=true}
   &ScaleColor(
    Scale Alpha=&FloatFromCurve(
     FloatCurve=DataInterface() {propertyValues {Curve {keys [
         <at 0 0 RCIM_Cubic>
         {arriveTangent=-0.0772, interpMode=RCIM_Cubic, leaveTangent=-0.0772, tangentMode=RCTM_User, time=0.3, value=1}
         {arriveTangent=-0.1048, interpMode=RCIM_Cubic, leaveTangent=-0.1048, tangentMode=RCTM_User, time=0.9, value=1}
         {arriveTangent=-25.4234, interpMode=RCIM_Cubic, leaveTangent=-25.4234, tangentMode=RCTM_User, time=1, value=0}
        ]}}}
     Scale Curve=0.3
    )
    Scale RGB=&MakeVectorFromLinearColorRGB(Color=&ColorFromCurve(ColorCurve=DataInterface() {propertyValues {
        AlphaCurve {keys []}
        BlueCurve {keys [<at 0 1.3>,<at 0.5508 3>,<at 0.9157 18>]}
        GreenCurve {keys [<at 0 0.4564>,<at 0.5508 1.4202>,<at 0.9157 12.4703>]}
        RedCurve {keys [<at 0 0.1508>,<at 0.5508 0.09>,<at 0.9157 1.8>]}
       }}))
    ScaleA=true
    ScaleRGB=true
   )
  ]
  rendererClasses [&NiagaraSpriteRendererProperties]
  renderers [
   {rendererClass=&NiagaraSpriteRendererProperties}
   {rendererClass=&NiagaraSpriteRendererProperties, rendererIndex=1}
  ]
 }
 FN_Debug {
  bEnabled=false
  emitterUpdateScript [
   &EmitterState()
   &SpawnBurst_Instantaneous(Spawn Count=1, Use Spawn Probability=false)
  ]
  particleSpawnScript [/Niagara/Modules/Spawn/Initialization/InitializeParticle(
    Lifetime=5
    Module.Write Color=true
    Module.Write Lifetime=true
    Module.Write Mass=true
    Module.Write Material Random=true
    Module.Write Position=false
    Module.Write Scale=false
    Module.Write SpriteRotation=false
    Module.Write SpriteSize=true
    Position=/Niagara/Functions/Localspace/SimulationPosition()
    Sprite Size=<44 44>
   )]
  particleUpdateScript [
   &ParticleState(Kill Particles When Lifetime Has Elapsed=false)
   &FN_Timer()
   /SetVariables(Particles.Debug=0.1) {isSetParameters=true}
   &DynamicMaterialParameters(
    Index 0 Param 1=$Particles.Debug
    Module.Param0WriteEnabled=true
    Module.Param1WriteEnabled=true
    Module.Param2WriteEnabled=true
    Module.Param3WriteEnabled=true
   )
   &CameraOffset(Camera Offset Amount=33)
  ]
  rendererClasses [&NiagaraSpriteRendererProperties]
  renderers [{rendererClass=&NiagaraSpriteRendererProperties}]
 }
 Flare {
  emitterUpdateScript [
   &EmitterState()
   &SpawnBurst_Instantaneous(Spawn Time=0.24, Use Spawn Probability=false, UseLoopCountLimit=true)
  ]
  particleSpawnScript [
   /Niagara/Modules/Spawn/Initialization/V2/InitializeParticle(
    Color Mode=::NewEnumerator0
    Lifetime=0.18
    Sprite Size Mode=::NewEnumerator3
    Uniform Sprite Size=45
    UsePositionOffset=false
    Write Material Random=true
   )
   &CameraOffset(Camera Offset Amount=15)
  ]
  particleUpdateScript [
   &ParticleState()
   &ScaleSpriteSize(
    Non-Uniform Curve Scale=<3 0.5>
    Non-Uniform Curve Sprite Scale=DataInterface() {propertyValues {
      XCurve {keys [{arriveTangent=1.8566, interpMode=RCIM_Cubic, leaveTangent=1.8566, tangentMode=RCTM_User, time=0, value=0}, <at 0.9 1 RCIM_Cubic>]}
      YCurve {keys [{tangentMode=RCTM_User, time=0, value=1}, {tangentMode=RCTM_User, time=1, value=1}]}
     }}
    Scale Sprite Size Mode=::NewEnumerator3
   ) {module=ScaleSpriteSize001}
   &ScaleColor(
    Scale Alpha=&FloatFromCurve(
     FloatCurve=DataInterface() {propertyValues {Curve {keys [
         <at 0 0 RCIM_Cubic>
         <at 0.3 1 RCIM_Cubic>
         {arriveTangent=-2.5221, interpMode=RCIM_Cubic, leaveTangent=-2.5221, tangentMode=RCTM_User, time=1, value=0}
        ]}}}
     Scale Curve=0.6
    )
    Scale RGB=&MakeVectorFromLinearColorRGB(Color=&ColorFromCurve(ColorCurve=DataInterface() {propertyValues {
        AlphaCurve {keys []}
        BlueCurve {keys [<at 0 40 RCIM_Cubic>,<at 0.0828 20>,<at 0.9244 0.2 RCIM_Cubic>]}
        GreenCurve {keys [<at 0 28.949 RCIM_Cubic>,<at 0.0828 9.176>,<at 0.9244 0.1 RCIM_Cubic>]}
        RedCurve {keys [<at 0 11.8233 RCIM_Cubic>,<at 0.0828 0.6>,<at 0.9244 0.1122 RCIM_Cubic>]}
       }}))
    ScaleA=true
    ScaleRGB=true
   )
  ]
  rendererClasses [&NiagaraSpriteRendererProperties]
  renderers [{rendererClass=&NiagaraSpriteRendererProperties}]
 }
 Glow_Back {
  emitterUpdateScript [
   &EmitterState()
   &SpawnBurst_Instantaneous(Spawn Time=0.13, Use Spawn Probability=false, UseLoopCountLimit=true)
  ]
  particleSpawnScript [
   &InitializeParticle(
    Color Mode=::NewEnumerator0
    Lifetime=0.5
    Sprite Rotation Mode=::NewEnumerator2
    Sprite Size Mode=::NewEnumerator3
    Uniform Sprite Size=80
    UsePositionOffset=false
    Write Material Random=true
   )
   &CameraOffset(Camera Offset Amount=10)
  ]
  particleUpdateScript [
   &ParticleState()
   &ScaleColor(
    Scale Alpha=&FloatFromCurve(
     FloatCurve=DataInterface() {propertyValues {Curve {keys [
         {arriveTangent=66.1622, interpMode=RCIM_Cubic, leaveTangent=66.1622, tangentMode=RCTM_User, time=0.38, value=0}
         {arriveTangent=-1.785, interpMode=RCIM_Cubic, leaveTangent=-1.785, tangentMode=RCTM_User, time=0.4, value=1}
         {arriveTangent=-0.6419, interpMode=RCIM_Cubic, leaveTangent=-0.6419, tangentMode=RCTM_User, time=1, value=0}
        ]}}}
     Scale Curve=0.2
    )
    Scale RGB=&MakeVectorFromLinearColorRGB(Color=&ColorFromCurve(ColorCurve=DataInterface() {propertyValues {
        AlphaCurve {keys []}
        BlueCurve {keys [<at 0.5689 10>,<at 1 0.3>]}
        GreenCurve {keys [<at 0.5689 4.75>,<at 1 0.1146>]}
        RedCurve {keys [<at 0.5689 1>,<at 1 0.098>]}
       }}))
    ScaleA=true
    ScaleRGB=true
   )
   /SetVariables(Particles.ColorSpriteDark:LinearColor=&MakeLinearColorFromVectorAndFloat(
     "Vector (RGB)"=&MakeVectorFromLinearColorRGB(Color=&ColorFromCurve(ColorCurve=DataInterface() {propertyValues {
         AlphaCurve {keys []}
         BlueCurve {keys [<at 0.0374 0.04 RCIM_Cubic>]}
         GreenCurve {keys [<at 0.0374 0.022 RCIM_Cubic>]}
         RedCurve {keys [<at 0.0374 0.0253 RCIM_Cubic>]}
        }}))
     "Float (Alpha)"=&FloatFromCurve(
      FloatCurve=DataInterface() {propertyValues {Curve {keys [
          {arriveTangent=27.807, interpMode=RCIM_Cubic, leaveTangent=27.807, tangentMode=RCTM_User, time=0, value=0}
          {arriveTangent=0.1104, interpMode=RCIM_Cubic, leaveTangent=0.1104, tangentMode=RCTM_User, time=0.1, value=1.5}
          {arriveTangent=-4.2032, interpMode=RCIM_Cubic, leaveTangent=-4.2032, tangentMode=RCTM_User, time=0.7, value=0}
         ]}}}
      Scale Curve=0.8
     )
    )) {isSetParameters=true}
   &ScaleSpriteSize(Uniform Curve Sprite Scale=DataInterface() {propertyValues {Curve {keys [{tangentMode=RCTM_User, time=0, value=0}, {tangentMode=RCTM_User, time=0.1, value=1}]}}})
   /SetVariables(Particles.SpriteSizeDark:Vector2f=&Multiply_Vector2D_ByFloat(Float=1.3, Vector2D=$Particles.SpriteSize)) {isSetParameters=true}
  ]
  rendererClasses [&NiagaraSpriteRendererProperties]
  renderers [
   {rendererClass=&NiagaraSpriteRendererProperties}
   {rendererClass=&NiagaraSpriteRendererProperties, rendererIndex=1}
  ]
 }
 Glow_Bright {
  emitterUpdateScript [
   &EmitterState()
   &SpawnBurst_Instantaneous(Spawn Time=0.27, Use Spawn Probability=false, UseLoopCountLimit=true)
  ]
  particleSpawnScript [
   &InitializeParticle(
    Color Mode=::NewEnumerator0
    Lifetime=0.2
    Sprite Rotation Mode=::NewEnumerator2
    Sprite Size Mode=::NewEnumerator3
    Uniform Sprite Size=20
    UsePositionOffset=false
    Write Material Random=true
   )
   &CameraOffset(Camera Offset Amount=5)
   /SetVariables(Particles.PositionLight:NiagaraPosition=&AddVectorToPosition(Position=$Particles.Position, Vector=<0 -20 0>)) {isSetParameters=true}
  ]
  particleUpdateScript [
   &ParticleState()
   &ScaleColor(
    Scale Alpha=&FloatFromCurve(
     FloatCurve=DataInterface() {propertyValues {Curve {keys [
         <at 0 0 RCIM_Cubic>
         <at 0.1 1 RCIM_Cubic>
         {arriveTangent=-1.8363, interpMode=RCIM_Cubic, leaveTangent=-1.8363, tangentMode=RCTM_User, time=1, value=0}
        ]}}}
     Scale Curve=0.65
    )
    Scale RGB=&MakeVectorFromLinearColorRGB(Color=&ColorFromCurve(ColorCurve=DataInterface() {propertyValues {
        AlphaCurve {keys []}
        BlueCurve {keys [<at 0.0113 1>,<at 0.1224 20 RCIM_Cubic>,<at 0.5394 1 RCIM_Cubic>]}
        GreenCurve {keys [<at 0.0113 0.4193>, {interpMode=RCIM_Cubic, leaveTangent=-0.0458, time=0.1224, value=10.2}, <at 0.5394 0.3808 RCIM_Cubic>]}
        RedCurve {keys [<at 0.0113 0.04>, {interpMode=RCIM_Cubic, leaveTangent=0.1598, time=0.1224, value=0.4}, <at 0.5394 0.1744 RCIM_Cubic>]}
       }}))
    ScaleA=true
    ScaleRGB=true
   )
   &ScaleSpriteSize(Uniform Curve Sprite Scale=DataInterface() {propertyValues {Curve {keys [{tangentMode=RCTM_User, time=0, value=0}, {tangentMode=RCTM_User, time=0.1, value=1}, <at 0.9 0.7>]}}})
  ]
  rendererClasses [&NiagaraSpriteRendererProperties, &NiagaraLightRendererProperties]
  renderers [
   {rendererClass=&NiagaraSpriteRendererProperties}
   {rendererClass=&NiagaraLightRendererProperties, rendererIndex=1}
  ]
 }
 GreenLight {
  bEnabled=false
  emitterSpawnScript [&SampleSkeletalMesh(
    AdditionalYawPitchRoll=false
    Bone Sampling Mode:ENiagaraBoneSamplingMode=::NewEnumerator4
    Interpolate Mesh Sampling=false
    Position Offset Coordinate Space:ENiagaraCoordinateSpace=::Local
    Sampled Position Coordinate Space:ENiagaraCoordinateSpace=::Local
    Skeletal Mesh:NiagaraDataInterfaceSkeletalMesh=$User.SkeletalMesh
   )]
  emitterUpdateScript [
   &EmitterState(
    Inactive Response=::NewEnumerator1
    Life Cycle Mode=::NewEnumerator1
    Loop Behavior=::NewEnumerator1
    Loop Duration Mode=::NewEnumerator1
    UseLoopDelay=false
   )
   &SpawnBurst_Instantaneous(Use Spawn Probability=false, UseLoopCountLimit=false)
  ]
  particleSpawnScript [&InitializeParticle(
    Color=<rgba 0.1 1 0 0.5>
    Lifetime=5
    Position Mode=::NewEnumerator0
    Sprite Size Mode=::NewEnumerator3
    Uniform Sprite Size=9
    Write Material Random=false
   )]
  particleUpdateScript [
   &ParticleState(Kill Particles When Lifetime Has Elapsed=false)
   &ScaleColor(
    Scale Alpha=&MakeCustomFloatFromBool(Bool=$User.bGreenLightEnabled, True Float=1, False Float=0)
    ScaleA=true
    ScaleRGB=false
   )
   &CameraOffset(Camera Offset Amount=9)
  ]
  rendererClasses [&NiagaraSpriteRendererProperties, &NiagaraMeshRendererProperties]
  renderers [
   {rendererClass=&NiagaraSpriteRendererProperties}
   {rendererClass=&NiagaraMeshRendererProperties, rendererIndex=1}
  ]
 }
 Halo {
  emitterUpdateScript [
   &EmitterState()
   &SpawnBurst_Instantaneous(Spawn Time=0.3, Use Spawn Probability=false, UseLoopCountLimit=true)
  ]
  particleSpawnScript [
   &InitializeParticle(
    Color Mode=::NewEnumerator0
    Lifetime=0.3
    Sprite Size Mode=::NewEnumerator3
    Uniform Sprite Size=50
    UsePositionOffset=false
    Write Material Random=true
   )
   &CameraOffset(Camera Offset Amount=35)
  ]
  particleUpdateScript [
   &ParticleState()
   &ScaleSpriteSize(
    Scale Sprite Size Mode=::NewEnumerator0
    Uniform Scale Factor=&FloatFromCurve(FloatCurve=DataInterface() {propertyValues {Curve {keys [{arriveTangent=0.4202, interpMode=RCIM_Cubic, leaveTangent=0.4202, tangentMode=RCTM_User, time=0, value=0.85}, <at 1 1 RCIM_Cubic>]}}})
   ) {module=ScaleSpriteSize001}
   &ScaleColor(
    Scale Alpha=&FloatFromCurve(
     FloatCurve=DataInterface() {propertyValues {Curve {keys [
         <at 0 0 RCIM_Cubic>
         <at 0.1 1 RCIM_Cubic>
         {arriveTangent=-0.7787, interpMode=RCIM_Cubic, leaveTangent=-0.7787, tangentMode=RCTM_User, time=1, value=0}
        ]}}}
     Scale Curve=0.4
    )
    Scale RGB=&MakeVectorFromLinearColorRGB(Color=&ColorFromCurve(ColorCurve=DataInterface() {propertyValues {
        AlphaCurve {keys []}
        BlueCurve {keys [<at 0 4 RCIM_Cubic>,<at 0.0828 13>,<at 0.9059 0.1 RCIM_Cubic>]}
        GreenCurve {keys [<at 0 2.8949 RCIM_Cubic>,<at 0.0828 7.1974>,<at 0.9059 0.05 RCIM_Cubic>]}
        RedCurve {keys [<at 0 1.1823 RCIM_Cubic>,<at 0.0828 2.6>,<at 0.9059 0.0561 RCIM_Cubic>]}
       }}))
    ScaleA=true
    ScaleRGB=true
   )
  ]
  rendererClasses [&NiagaraSpriteRendererProperties]
  renderers [{rendererClass=&NiagaraSpriteRendererProperties}]
 }
 Sparks {
  emitterUpdateScript [
   &EmitterState()
   &SpawnBurst_Instantaneous(Spawn Count=16, Spawn Time=0.34, Use Spawn Probability=false, UseLoopCountLimit=true)
  ]
  particleSpawnScript [
   &InitializeParticle(
    Color Mode=::NewEnumerator0
    Lifetime=&RandomRangeFloat(Minimum=0.1, Maximum=0.5)
    Sprite Size Mode=::NewEnumerator3
    Uniform Sprite Size=&RandomRangeFloat(Minimum=1, Maximum=5)
    UsePositionOffset=false
    Write Material Random=true
   )
   &ShapeLocation(MaskSpawnGroup=false, Sphere Distribution=::NewEnumerator3, Sphere Radius=7, UseOwnerScale=true)
   &AddVelocity(
    Module.Use Falloff Distance=false
    Override Seed=false
    Velocity Mode=::NewEnumerator1
    Velocity Speed=&RandomRangeFloat(Minimum=250, Maximum=400)
   )
   &CameraOffset(Camera Offset Amount=10)
  ]
  particleUpdateScript [
   &InheritSourceMovement(Applied Position Scale=<0.8 0.8 0.8>, ApplyForce=false, ApplyPosition=true, ApplyPositionFalloff=false)
   &ParticleState()
   &ScaleColor(
    Scale Alpha=&FloatFromCurve(
     FloatCurve=DataInterface() {propertyValues {Curve {keys [
         <at 0 0 RCIM_Cubic>
         {arriveTangent=0.0417, interpMode=RCIM_Cubic, leaveTangent=0.0417, tangentMode=RCTM_User, time=0.1, value=1}
         <at 0.4 1>
         <at 1 0 RCIM_Cubic>
        ]}}}
     Scale Curve=0.3
    )
    Scale RGB=&MakeVectorFromLinearColorRGB(Color=&ColorFromCurve(ColorCurve=DataInterface() {propertyValues {
        AlphaCurve {keys []}
        BlueCurve {keys [<at 0 20>,<at 0.215 15>,<at 0.8495 1>]}
        GreenCurve {keys [<at 0 15.4526>,<at 0.215 5.1>,<at 0.8495 0.28>]}
        RedCurve {keys [<at 0 14>,<at 0.215 0.15>,<at 0.8495 0.05>]}
       }}))
    ScaleA=true
    ScaleRGB=true
   )
   &ScaleSpriteSize(
    Non-Uniform Curve Sprite Scale=DataInterface() {propertyValues {
      XCurve {keys [{tangentMode=RCTM_User, time=0.6, value=1}, {tangentMode=RCTM_User, time=1, value=0.1}]}
      YCurve {keys [{tangentMode=RCTM_User, time=0, value=1}, {tangentMode=RCTM_User, time=1, value=1}]}
     }}
    Scale Sprite Size Mode=::NewEnumerator3
   )
   &Drag(Drag=7, Use Linear Drag=true, Use Rotational Drag=true)
   &GravityForce(Gravity=<0 0 -200>)
   &SolveForcesAndVelocity(Clamp Velocity=false, Limit Acceleration=false)
   &ScaleSpriteSizeBySpeed(Max Scale Factor=<0.5 6>, Min Scale Factor=<0.3 1>, Velocity Threshold=500)
  ]
  rendererClasses [&NiagaraSpriteRendererProperties]
  renderers [{rendererClass=&NiagaraSpriteRendererProperties}]
 }
}
summary {
 systemName=NS_Quest_S24_Narrative_Q3_BackSensor
 emitters [
  {bEnabled=false, emitterName=FN_Debug, rendererClasses [&NiagaraSpriteRendererProperties]}
  {
   bEnabled=false
   emitterName=GreenLight
   rendererClasses [&NiagaraSpriteRendererProperties, &NiagaraMeshRendererProperties]
  }
  {
   bEnabled=false
   emitterName=BlueLight
   rendererClasses [&NiagaraSpriteRendererProperties, &NiagaraMeshRendererProperties]
  }
  {emitterName=Glow_Back, rendererClasses [&NiagaraSpriteRendererProperties]}
  {emitterName=Sparks, rendererClasses [&NiagaraSpriteRendererProperties]}
  {emitterName=Glow_Bright, rendererClasses [&NiagaraSpriteRendererProperties, &NiagaraLightRendererProperties]}
  {emitterName=Flare, rendererClasses [&NiagaraSpriteRendererProperties]}
  {emitterName=Halo, rendererClasses [&NiagaraSpriteRendererProperties]}
  {emitterName=Disk, rendererClasses [&NiagaraSpriteRendererProperties]}
 ]
 userVariables [
  User.bGreenLightEnabled {type=NiagaraBool, defaultValue=true}
  User.BlinkSpeed {type=NiagaraFloat, defaultValue=0, description=""}
  User.ClosestObjectPosition {type=NiagaraPosition, description="", defaultValue {}}
  User.GreenLightToggleDistance {type=NiagaraFloat, defaultValue=350, description=""}
  User.SkeletalMesh {
   type=NiagaraDataInterfaceSkeletalMesh
   defaultValue {
    struct=NiagaraExt_VariableValue_DataInterface
    value {dataInterface=&self:NiagaraDataInterfaceSkeletalMesh_0}
   }
  }
 ]
}
